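# Material function that wraps MF_TriplanarVirtualBox, exposing its inputs as material parameters in the MF_TVB_Wrap group.
# It supplies a texture, UV tiling and offset built from paired scalars, a center point offset, depth, blend contrast and a static local camera vector switch.
# It passes through the inner function's result, the three per-axis UVs and the Y and Z blend values as outputs.
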MaterialFunction /Game/Global/Materials/MaterialGym/Materials/MaterialElements/UVs/PreskinnedVirtualBox/MF_TriplanarVirtualBox_Wrap
graph {nodes [
  n1 = MaterialFunctionCall!(
   TVB_Texture=TextureObjectParameter() {group=MF_TVB_Wrap, parameterName=TVB_Texture, texture=&T_BlacksRGB}
   TVB_UV_Tiling=AppendVector(
    A=ScalarParameter() {defaultValue=2, group=MF_TVB_Wrap, parameterName=TVB_Tiling_UVX}
    B=ScalarParameter() {defaultValue=2, group=MF_TVB_Wrap, parameterName=TVB_Tiling_UVY}
   )
   TVB_UV_Offset=AppendVector(
    A=ScalarParameter() {defaultValue=0.004, group=MF_TVB_Wrap, parameterName=TVB_Offset_UVX}
    B=ScalarParameter() {defaultValue=0.5, group=MF_TVB_Wrap, parameterName=TVB_Offset_UVY}
   )
   TVB_CenterPointOffset=VectorParameter() {group=MF_TVB_Wrap, parameterName=TVB_CenterPointOffset, defaultValue=<rgba 0 0 0 1>}.RGB
   TVB_Depth=ScalarParameter() {defaultValue=100, group=MF_TVB_Wrap, parameterName=TVB_Depth}
   TVB_BlendContrast=ScalarParameter() {defaultValue=5, group=MF_TVB_Wrap, parameterName=TVB_BlendContrast}
   TVB_UseLocalCameraVector=StaticBoolParameter() {group=MF_TVB_Wrap, parameterName=TVB_UseLocalCameraVector}
  ) {materialFunction=&/MF_TriplanarVirtualBox}
  FunctionOutput(None=@n1.Result) {outputName=TVB_Result}
  FunctionOutput(None=@n1.UV X) {outputName=TVB_UVX, sortPriority=1}
  FunctionOutput(None=@n1.UV Y) {outputName=TVB_UVY, sortPriority=2}
  FunctionOutput(None=@n1.UV Z) {outputName=TVB_UVZ, sortPriority=3}
  FunctionOutput(None=@n1.UV Blend Y) {outputName=TVB_UVBlendY, sortPriority=4}
  FunctionOutput(None=@n1.UV Blend Z) {outputName=TVB_UVBlendZ, sortPriority=5}
 ]}
