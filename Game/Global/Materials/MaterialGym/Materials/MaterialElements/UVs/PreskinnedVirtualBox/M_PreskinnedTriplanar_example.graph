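# Example skeletal mesh material that demonstrates preskinned triplanar virtual box mapping with animated storm noise.
# MF_TriplanarVirtualBox_Wrap supplies three axis UV sets and two blend masks through named reroutes.
# Polar UV wrappers feed a storm noise function to produce separate X, Y and Z noise results.
# Two nested lerps, fed by named reroutes, combine the projections, avoiding reliance on a single UV layout.
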Material /Game/Global/Materials/MaterialGym/Materials/MaterialElements/UVs/PreskinnedVirtualBox/M_PreskinnedTriplanar_example
graph {
 properties {bAutomaticallySetUsageInEditor=false, bUsedWithMeshDeformer=true, bUsedWithSkeletalMesh=true, usedShadingModels=Default Lit}
 nodes [
  Constant()
  LinearInterpolate(
   A=LinearInterpolate(A=NamedRerouteUsage(), B=NamedRerouteUsage(), Alpha=NamedRerouteUsage())
   B=NamedRerouteUsage()
   Alpha=NamedRerouteUsage()
  )
  n1 = MaterialFunctionCall() {materialFunction=&/MF_TriplanarVirtualBox_Wrap}
  NamedRerouteDeclaration(None=@n1.TVB_UVX) {name=UVX, nodeColor=<rgba 0 1 0.4375 1>}
  NamedRerouteDeclaration(None=@n1.TVB_UVY) {name=UVY, nodeColor=<rgba 1 0.1172 0 1>}
  NamedRerouteDeclaration(None=@n1.TVB_UVZ) {name=UVZ, nodeColor=<rgba 0 0.9531 1 1>}
  NamedRerouteDeclaration(None=@n1.TVB_UVBlendY) {name=UVBlendY, nodeColor=<rgba 1 0.7266 0 1>}
  NamedRerouteDeclaration(None=@n1.TVB_UVBlendZ) {name=UVBlendZ, nodeColor=<rgba 0.4062 0 1 1>}
  n2 = NamedRerouteUsage()
  NamedRerouteDeclaration(None=MaterialFunctionCall!(InputUV=MaterialFunctionCall!(UVInput=@n2, PolarUV_CenterOffsetR, PolarUV_CenterOffsetG) {materialFunction=/Game/Global/Materials/MaterialGym/Materials/MaterialElements/UVs/UVMainpulation/PolarUV/MF_Polar_Wrap}.OutputUV) {materialFunction=/Game/Global/Materials/MaterialGym/Materials/MaterialElements/UVs/AnimatedNoise/MF_StormNoise_Wrap}.Result) {name=NoiseX, nodeColor=<rgba 0 1 0.0859 1>}
  n3 = NamedRerouteUsage()
  NamedRerouteDeclaration(None=MaterialFunctionCall(InputUV=MaterialFunctionCall(UVInput=@n3) {materialFunction=&MF_Polar_Wrap}.OutputUV) {materialFunction=&MF_StormNoise_Wrap}.Result) {name=NoiseY, nodeColor=<rgba 1 0 0.2344 1>}
  n4 = NamedRerouteUsage()
  NamedRerouteDeclaration(None=MaterialFunctionCall(InputUV=MaterialFunctionCall(UVInput=@n4) {materialFunction=&MF_Polar_Wrap}.OutputUV) {materialFunction=&MF_StormNoise_Wrap}.Result) {name=NoiseZ, nodeColor=<rgba 0 0.5547 1 1>}
 ]
}
