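# Master material that wraps the triplanar virtual box material function, likely projecting textures from a virtual box on pre-skinned positions, as its folder name suggests.
# The graph holds a single MF_TriplanarVirtualBox_Wrap function call, which carries all of the shading logic.
# It uses Default Lit shading and is flagged for skeletal meshes and mesh deformers, with automatic usage detection in the editor turned off.
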Material /Game/Global/Materials/MaterialGym/Materials/MaterialElements/UVs/PreskinnedVirtualBox/M_TriplanarVirtualBox
graph {
 properties {bAutomaticallySetUsageInEditor=false, bUsedWithMeshDeformer=true, bUsedWithSkeletalMesh=true, usedShadingModels=Default Lit}
 nodes [MaterialFunctionCall() {materialFunction=&/MF_TriplanarVirtualBox_Wrap}]
}
